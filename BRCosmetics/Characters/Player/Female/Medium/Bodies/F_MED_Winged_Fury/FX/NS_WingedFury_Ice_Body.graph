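# Ice-variant body effect for the Winged Fury player cosmetic, emitting blue glowing sprites and lights from the character's skeletal mesh.
# Gloves_Flames layers two spawn groups on bones, Heels pins a mesh, glow and pulsing light to sockets, and AshSprites and Empty add falling and rising particles.
# User hand and foot hidden flags drive visibility tags, with foot hiding also disabling heel glow and lights.
# Light specular differs between front end and gameplay.
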NiagaraSystem /BRCosmetics/Characters/Player/Female/Medium/Bodies/F_MED_Winged_Fury/FX/NS_WingedFury_Ice_Body
dependencies {
 usedDataInterfaces [
  &NiagaraDataInterfaceCurve
  &NiagaraDataInterfaceSpriteRendererInfo
  &NiagaraDataInterfaceVectorCurve
  &NiagaraDataInterfaceColorCurve
  &NiagaraDataInterfaceSkeletalMesh
 ]
 usedDynamicInputs []
 usedModules [
  &self:SystemSpawnScript.SystemScriptSource.SystemScriptGraph.NiagaraNodeAssignment_0.SetVariables_#1
  &SystemState
 ]
}
emitters {
 AshSprites {
  emitterUpdateScript [
   &EmitterState()
   &SpawnBurst_Instantaneous(
    Spawn Count=&RandomRangeInt(Minimum=0, Maximum=1)
    Spawn Time=0.4
    Use Spawn Probability=false
    UseLoopCountLimit=false
   )
   &SpawnRate(SpawnRate=8, Use Spawn Probability=false)
  ]
  particleSpawnScript [
   &InitializeParticle(
    Lifetime Max=0.75
    Lifetime Min=0.3
    Lifetime Mode=::NewEnumerator1
    Sprite Rotation Mode=::NewEnumerator2
    Sprite Size Mode=::NewEnumerator1
    Uniform Sprite Size Max=2
    Uniform Sprite Size Min=1.1
    UsePositionOffset=false
    Write Material Random=true
   )
   &InheritVelocity(Inherited Velocity Amount Scale=<0.1 0.1 0.1>, Speed Limit=true, UseSpeedThreshold=false)
   &AddVelocity(Velocity=&RandomRangeVector(Minimum=<-4 -4 -20>, Maximum=<4 4 10>))
   &DynamicMaterialParameters(
    Index 0 Param 1=&RandomRangeFloat(Minimum=0, Maximum=1)
    Index 0 Param 2=&RandomRangeFloat(Minimum=0, Maximum=1)
    Index 0 Param 3=&RandomRangeFloat(Minimum=0.25, Maximum=1)
    Index 0 Param 4=&RandomRangeFloat(Minimum=0.25, Maximum=1.25)
    Module.Index0Param0Write=true
    Module.Index0Param1Write=true
    Module.Index0Param2Write=true
    Module.Index0Param3Write=true
   )
   &CameraOffset(Camera Offset Amount=5)
  ]
  particleUpdateScript [
   &ParticleState()
   &AccelerationForce(Acceleration=&RandomRangeVector(Minimum=<0 0 -20>, Maximum=<0 0 -50>))
   &Drag(Drag=1.5, Use Linear Drag=true, Use Rotational Drag=true)
   &SolveForcesAndVelocity(Clamp Velocity=false, Limit Acceleration=false)
   &ScaleSpriteSizeBySpeed(Max Scale Factor=<10 20>)
   &SpriteRotationRate(Rotation Rate=&RandomRangeFloat(Minimum=-1, Maximum=1))
   &SkeletalMeshLocation(
    Bone / Socket Index=&RandomRangeInt(Minimum=0, Maximum=3)
    Bone Sampling Mode=::NewEnumerator4
    MaskSpawnGroup=false
    Skeletal Mesh=$System.Skeletal Mesh
   )
   &SubUVAnimation(
    Evaluation Type=::NewEnumerator0
    "Sprite Renderer "=DataInterface() {propertyValues {SpriteRenderer=&self:AshSprites_0.NiagaraSpriteRendererProperties_0}}
    SubUV Animation Mode:ENiagara_SubUVLookupModeV2=::NewEnumerator2
    UseEndFrame=false
    UseStartFrame=false
   )
   &ScaleColor(
    Scale Alpha=&FloatFromCurve(FloatCurve=DataInterface() {propertyValues {Curve {keys [<at 0.75 1 RCIM_Cubic>,<at 1 0>]}}})
    Scale RGB=&VectorFromCurve(VectorCurve=DataInterface() {propertyValues {
       XCurve {keys [<at 0 2.5 RCIM_Cubic>,<at 0.7 0 RCIM_Cubic>]}
       YCurve {keys [<at 0 15 RCIM_Cubic>,<at 0.7 0 RCIM_Cubic>]}
       ZCurve {keys [<at 0 50 RCIM_Cubic>,<at 0.7 0 RCIM_Cubic>]}
      }})
    ScaleA=true
    ScaleRGB=true
   )
   /SetVariables(Particles.VisibilityTag=&MakeCustomIntFromBool(Bool=$User.bisFootHidden, True INT=1, False INT=0)) {isSetParameters=true}
  ]
  rendererClasses [&NiagaraSpriteRendererProperties]
  renderers [{rendererClass=&NiagaraSpriteRendererProperties}]
 }
 Empty {
  emitterUpdateScript [
   &EmitterState()
   &SpawnRate(SpawnRate=4, Use Spawn Probability=false)
  ]
  particleSpawnScript [
   &PartitionParticles()
   &InitializeParticle(
    Color=<rgba 0.55 0.4987 2 1>
    Lifetime Max=1
    Lifetime Min=0.4
    Lifetime Mode=::NewEnumerator1
    Sprite Rotation Mode=::NewEnumerator2
    Sprite Size Mode=::NewEnumerator1
    Uniform Sprite Size Max=2.2
    Uniform Sprite Size Min=1
    UsePositionOffset=false
    Write Material Random=true
   )
   &InheritVelocity(Inherited Velocity Amount Scale=<0.25 0.25 0.25>, Speed Limit=true, UseSpeedThreshold=false)
   &AddVelocity(Velocity=&RandomRangeVector(Minimum=<0 0 10>, Maximum=<0 0 50>))
   &SkeletalMeshLocation(
    Bone / Socket Index=$Particles.PartitionParticles.Partition
    MaskSpawnGroup=false
    Mesh Sampling Type=::NewEnumerator5
    Skeletal Mesh=$System.Skeletal Mesh
    Socket Sampling Mode=::NewEnumerator5
   )
   &ShapeLocation(MaskSpawnGroup=false, Override Seed=false, Sphere Radius=3, UseOwnerScale=true)
  ]
  particleUpdateScript [
   &ParticleState()
   &SpriteRotationRate(Rotation Rate=&RandomRangeFloat(Minimum=-1, Maximum=1))
   &ScaleSpriteSize(Uniform Curve Sprite Scale=DataInterface() {propertyValues {Curve {keys [
        <at 0 0 RCIM_Cubic>
        {arriveTangent=1.4286, interpMode=RCIM_Cubic, leaveTangent=1.4286, time=0.12, value=1}
        {arriveTangent=-1.1364, interpMode=RCIM_Cubic, leaveTangent=-1.1364, time=0.7, value=1}
        {interpMode=RCIM_Cubic, leaveTangent=-4.9665, time=1, value=0}
       ]}}})
   &Drag(Use Linear Drag=true, Use Rotational Drag=true)
   &SolveForcesAndVelocity(Clamp Velocity=false, Limit Acceleration=false)
   &ScaleColor(
    Scale Alpha=&FloatFromCurve(FloatCurve=DataInterface() {propertyValues {Curve {keys [<at 0.75 1 RCIM_Cubic>,<at 1 0>]}}})
    Scale RGB=&VectorFromCurve(VectorCurve=DataInterface() {propertyValues {
       XCurve {keys [<at 0 4.6 RCIM_Cubic>,<at 1 2 RCIM_Cubic>]}
       YCurve {keys [<at 0 21.6011 RCIM_Cubic>,<at 1 0.0852 RCIM_Cubic>]}
       ZCurve {keys [<at 0 100 RCIM_Cubic>,<at 1 0 RCIM_Cubic>]}
      }})
    ScaleA=true
    ScaleRGB=true
   )
   /SetVariables(Particles.VisibilityTag=&MakeCustomIntFromBool(Bool=$User.bisFootHidden, True INT=1, False INT=0)) {isSetParameters=true}
  ]
  rendererClasses [&NiagaraSpriteRendererProperties]
  renderers [{rendererClass=&NiagaraSpriteRendererProperties}]
 }
 Gloves_Flames {
  emitterUpdateScript [
   &EmitterState()
   &SpawnRate(SpawnRate=8, Use Spawn Probability=false)
   &SpawnRate(SpawnGroup=1, SpawnRate=8, Use Spawn Probability=false) {module=SpawnRate001}
  ]
  particleSpawnScript [
   /SetVariables(Particles.VisibilityTag=&SpawnGroup()) {isSetParameters=true}
   &InitializeParticle(
    Color=&MaskLinearColorBySpawnGroup(Masked Linear Color=<rgba 0 0.2472 0.8802 1>, Passthrough Linear Color=<rgba 0.01 0.9 3 0.5>)
    Lifetime=&MaskFloatBySpawnGroup(
     Masked Float=&RandomRangeFloat(Minimum=1.5, Maximum=3)
     Passthrough Float=&RandomRangeFloat(Minimum=0.65, Maximum=1)
    )
    Sprite Rotation Angle=&MaskFloatBySpawnGroup(Masked Float=0, Passthrough Float=&RandomRangeFloat(Minimum=0, Maximum=1))
    Sprite Rotation Mode=::NewEnumerator1
    Sprite Size Mode=::NewEnumerator3
    Uniform Sprite Size=&MaskFloatBySpawnGroup(
     Masked Float=&RandomRangeFloat(Minimum=20, Maximum=30)
     Passthrough Float=&RandomRangeFloat(Minimum=11, Maximum=15)
    )
    UsePositionOffset=false
    Write Material Random=true
   )
   &DynamicMaterialParameters(
    Index 0 Param 1=&MaskFloatBySpawnGroup(Masked Float=1.7, Passthrough Float=0)
    Index 0 Param 2=1
    Index 0 Param 3=1
    Index 0 Param 4=&MaskFloatBySpawnGroup(Masked Float=&RandomRangeFloat(Minimum=0, Maximum=1), Passthrough Float=1)
    Module.Index0Param0Write=true
    Module.Index0Param1Write=true
    Module.Index0Param2Write=true
    Module.Index0Param3Write=true
   )
  ]
  particleUpdateScript [
   &ParticleState()
   &SkeletalMeshLocation(
    Bone / Socket Index=&RandomRangeInt(Minimum=10, Maximum=11)
    Bone Sampling Mode=::NewEnumerator4
    MaskSpawnGroup=false
    Skeletal Mesh=$System.Skeletal Mesh
   )
   &CameraOffset(Camera Offset Amount=&MaskFloatBySpawnGroup(
     Masked Float=20
     Passthrough Float=&FloatFromCurve(FloatCurve=DataInterface() {propertyValues {Curve {keys [{arriveTangent=32.2774, interpMode=RCIM_Cubic, time=0, value=0}, <at 0.45 5.7565>]}}})
    ))
   &SubUVAnimation(
    "Sprite Renderer "=DataInterface() {propertyValues {SpriteRenderer=&self:Gloves_Flames_1.NiagaraSpriteRendererProperties_1}}
    UseEndFrame=false
    UseStartFrame=false
   )
   &ScaleColor(
    Scale Alpha=&MaskFloatBySpawnGroup(
     Masked Float=&FloatFromCurve(FloatCurve=DataInterface() {propertyValues {Curve {keys [
          <at -0.0198 0.7593 RCIM_Cubic>
          {arriveTangent=0.4632, interpMode=RCIM_Cubic, leaveTangent=0.4632, time=0.1852, value=0.9331}
          {arriveTangent=-0.4766, interpMode=RCIM_Cubic, leaveTangent=-0.4766, time=0.5, value=1}
          {arriveTangent=-2, interpMode=RCIM_Cubic, leaveTangent=-2, time=0.8, value=0.64}
          {arriveTangent=-4.0865, leaveTangent=-4.0865, time=1, value=0}
         ]}}})
     Passthrough Float=&FloatFromCurve(FloatCurve=DataInterface() {propertyValues {Curve {keys [{arriveTangent=11.7282, interpMode=RCIM_Cubic, time=0, value=0}, <at 0.2 1 RCIM_Cubic>, <at 1 0>]}}})
    )
    Scale RGB=&MaskVectorBySpawnGroup(
     Masked Vector=<1 1 1>
     Passthrough Vector=&VectorFromCurve(VectorCurve=DataInterface() {propertyValues {XCurve {keys [<at 0 4>,<at 1 1>]}, YCurve {keys [<at 0 4>,<at 1 1>]}, ZCurve {keys [<at 0 4>,<at 1 1>]}}})
    )
    ScaleA=true
    ScaleRGB=true
   )
   /SetVariables() {isSetParameters=true}
   /SetVariables(Particles.VisibilityTag=&MakeCustomIntFromBool(Bool=$User.bisHandHidden, True INT=2, False INT=$Particles.Initial.VisibilityTag)) {isSetParameters=true}
  ]
  rendererClasses [&NiagaraSpriteRendererProperties]
  renderers [
   {rendererClass=&NiagaraSpriteRendererProperties}
   {rendererClass=&NiagaraSpriteRendererProperties, rendererIndex=1}
  ]
 }
 Heels {
  emitterUpdateScript [
   &EmitterState()
   &SpawnBurst_Instantaneous(Spawn Count=2, Use Spawn Probability=false, UseLoopCountLimit=true)
  ]
  particleSpawnScript [
   &PartitionParticles()
   &InitializeParticle(
    Color=<rgba 0.3385 0.6449 1 0.5>
    Sprite Size Mode=::NewEnumerator3
    Uniform Sprite Size=30
    UsePositionOffset=false
    Write Material Random=true
   )
   &DynamicMaterialParameters(
    Index 0 Param 1=4
    Index 0 Param 2=1
    Module.Index0Param0Write=true
    Module.Index0Param1Write=true
    Module.Index0Param2Write=false
    Module.Index0Param3Write=false
   )
   /SetVariables(Particles.Glow Color=<rgba 0.126 0.5308 1 0.5>) {isSetParameters=true}
   &CameraOffset(Camera Offset Amount=5)
  ]
  particleUpdateScript [
   &ParticleState(Kill Particles When Lifetime Has Elapsed=false, Let Infinitely Lived Particles Die When Emitter Deactivates=true)
   &SkeletalMeshLocation(
    Bone / Socket Index=$Particles.PartitionParticles.Partition
    MaskSpawnGroup=false
    Mesh Sampling Type=::NewEnumerator5
    Sampled Position Offset=<0 2 -1>
    Skeletal Mesh=$System.Skeletal Mesh
    Socket Sampling Mode=::NewEnumerator5
   )
   /SetVariables(
    Particles.Glow Camera Offset=20
    Particles.Glow Color=&MultiplyLinearColor(
     Linear Color=<rgba 0.126 0.5308 1 0.5>
     Scale Factor=&MakeFloat4FromVectorAndFloat(
      "Vector (RGB)"=<1 1 1>
      "Float (Alpha)"=&MakeCustomFloatFromBool(Bool=$User.bisFootHidden, True Float=0, False Float=1)
     )
    )
    Particles.Glow Position:NiagaraPosition=&AddVectorToPosition(Position=$Particles.SkeletalMeshLocation.SampledPosition, Vector=<0 -1 4>)
   ) {isSetParameters=true}
   /SetVariables(
    Particles.LightColor=&ScaleLinearColorByCurve(
     InitialValue=<rgba 0 1.9064 5 1>
     ScaleCurve=DataInterface() {propertyValues {
       AlphaCurve {keys [<at 0 1>]}
       BlueCurve {keys [<at 0 1.5>,<at 0.5 1>,<at 1 1.5>]}
       GreenCurve {keys [<at 0 1.5>,<at 0.5 1>,<at 1 1.5>]}
       RedCurve {keys [<at 0 1.5>,<at 0.5 1>,<at 1 1.5>]}
      }}
     CurveIndex=$Emitter.NormalizedLoopAge
    )
    Particles.LightEnabled:NiagaraBool=&Boolean_NotOperation(Bool=$User.bisFootHidden)
    Particles.LightExponent=3
    Particles.LightPosition:NiagaraPosition=&AddVectorToPosition(Position=$Particles.Position, Vector=<0 -5 10>)
    Particles.LightRadius=30
    Particles.LightSpecularScale:NiagaraFloat=&FN_FrontEndFloatModifier(Gameplay=0.25, Front End=0)
   ) {isSetParameters=true}
   &ScaleColor(ScaleA=true, ScaleRGB=true)
   /SetVariables(Particles.VisibilityTag=&MakeCustomIntFromBool(Bool=$User.bisFootHidden, True INT=1, False INT=0)) {isSetParameters=true}
  ]
  rendererClasses [
   &NiagaraMeshRendererProperties
   &NiagaraSpriteRendererProperties
   &NiagaraLightRendererProperties
  ]
  renderers [
   {rendererClass=&NiagaraMeshRendererProperties, rendererIndex=0}
   {rendererClass=&NiagaraSpriteRendererProperties, rendererIndex=1}
   {rendererClass=&NiagaraLightRendererProperties, rendererIndex=2}
  ]
 }
}
summary {
 systemName=NS_WingedFury_Ice_Body
 emitters [
  {emitterName=AshSprites, rendererClasses [&NiagaraSpriteRendererProperties]}
  {emitterName=Gloves_Flames, rendererClasses [&NiagaraSpriteRendererProperties]}
  {
   emitterName=Heels
   rendererClasses [
    &NiagaraMeshRendererProperties
    &NiagaraSpriteRendererProperties
    &NiagaraLightRendererProperties
   ]
  }
  {emitterName=Empty, rendererClasses [&NiagaraSpriteRendererProperties]}
 ]
 userVariables [
  User.bisFootHidden {type=NiagaraBool}
  User.bIsFrontEnd {type=NiagaraBool}
  User.bIsFrontEndPreview {type=NiagaraBool}
  User.bisHandHidden {type=NiagaraBool}
 ]
}
