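# Physics asset for the Winged Fury female medium character's face accessory mesh, defining bodies for the ponytail, front and side hair, and earrings.
# The ponytail chains from the head through two base boxes and four capsules, then branches into four tips.
# Side hair hangs directly from the head, while earrings and front hair use base bodies with free angular constraints.
# Head, neck and spine capsules form the body volume, linked by free angular constraints.
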PhysicsAsset /BRCosmetics/Characters/Player/Female/Medium/Bodies/F_MED_Winged_Fury/Meshes/Parts/F_MED_Winged_Fury_FaceAcc_Physics
bodies {
 dyn_earring_base_l [{shapeType=Capsule, length=0.25, radius=0.5}]
 dyn_earring_base_r [{shapeType=Capsule, length=0.25, radius=0.5}]
 dyn_earring_l [{shapeType=Box, extentX=0.5, extentY=3.5, extentZ=3.5, center=<0 1.6895 -0.0001>}]
 dyn_earring_r [{shapeType=Box, extentX=0.5, extentY=3.5, extentZ=3.5, center=<0 -1.6895 0.0001>, rotation=<rot 0 0 180>}]
 dyn_hairFront [{shapeType=Box, extentX=3, extentY=1.5, extentZ=7, center=<1.5751 0 0>}]
 dyn_hairFront_base [{shapeType=Capsule, length=0, radius=1}]
 dyn_hairSide_l [{shapeType=Box, extentX=7.5, extentY=11, extentZ=2.5, center=<3.1876 0 0>}]
 dyn_hairSide_r [{shapeType=Box, extentX=7.5, extentY=11, extentZ=2.5, center=<-3.1876 0 0>}]
 dyn_hair_ponyTail_tipA [{shapeType=Capsule, length=1, radius=1.25, center=<1.6286 0 0>, rotation=<rot 90 0 0>}]
 dyn_hair_ponyTail_tipb [{shapeType=Capsule, length=2.5, radius=1.25, center=<2.5235 0 0>, rotation=<rot 90 0 0>}]
 dyn_hair_ponyTail_tipc [{shapeType=Capsule, length=1, radius=1.25, center=<1.5981 0 0>, rotation=<rot 90 0 0>}]
 dyn_hair_ponyTail_tipd [{shapeType=Capsule, length=1.5, radius=1, center=<1.5468 0 0>, rotation=<rot 90 0 0>}]
 dyn_ponyTail_A [{shapeType=Capsule, length=2.25, radius=2.5, center=<3.185 0 0.3197>, rotation=<rot 90 0 0>}]
 dyn_ponyTail_B [{shapeType=Capsule, length=2, radius=2.25, center=<2.8142 0 0>, rotation=<rot 90 0 0>}]
 dyn_ponyTail_C [{shapeType=Capsule, length=1.75, radius=2.25, center=<2.8493 0 0>, rotation=<rot 90 0 0>}]
 dyn_ponyTail_D [{shapeType=Capsule, length=1, radius=3.25, center=<3.5142 0 0>, rotation=<rot 90 0 0>}]
 dyn_ponyTaileBase_A [{shapeType=Box, extentX=5.5, extentY=15.5, extentZ=3, center=<3.1527 0 -0.2364>, rotation=<rot -2.8623 0 0>}]
 dyn_ponyTaileBase_B [{shapeType=Box, extentX=6, extentY=9.5, extentZ=3.5, center=<3.0538 0.0001 0>, rotation=<rot -4.0551 0 0>}]
 head [{shapeType=Capsule, length=4, radius=6.5, center=<2.4095 1.6993 0>, rotation=<rot 90 0 0>}]
 neck_01 [{shapeType=Capsule, length=7.75, radius=6, center=<2.6844 0.1504 0.0001>, rotation=<rot 90 0 -1.0421>}]
 spine_01 [{shapeType=Capsule, length=50.75, radius=9.25, center=<1.1016 1.8128 0>}]
 spine_03 [{shapeType=Capsule, length=50.75, radius=8.75, center=<0.6723 1.4377 0>}]
 spine_04 [{shapeType=Capsule, length=54.5, radius=9.5, center=<1.9871 0.1342 0>}]
 spine_05 [
  {shapeType=Capsule, length=57.75, radius=10.5, center=<1.4914 0.9972 0>}
  {shapeType=Capsule, length=42.5, radius=7.5, center=<9.8119 1.0274 0.0001>, rotation=<rot 0 7.4331 0>}
 ]
}
constraints [
 {
  bone1Name=head
  bone2Name=neck_01
  swing1LimitDegrees=45
  swing1Motion=Free
  swing2LimitDegrees=45
  swing2Motion=Free
  twistLimitDegrees=45
  twistMotion=Free
 }
 {
  bone1Name=neck_01
  bone2Name=spine_05
  swing1LimitDegrees=45
  swing1Motion=Free
  swing2LimitDegrees=45
  swing2Motion=Free
  twistLimitDegrees=45
  twistMotion=Free
 }
 {
  bone1Name=spine_05
  bone2Name=spine_04
  swing1LimitDegrees=45
  swing1Motion=Free
  swing2LimitDegrees=45
  swing2Motion=Free
  twistLimitDegrees=45
  twistMotion=Free
 }
 {
  bone1Name=spine_04
  bone2Name=spine_03
  swing1LimitDegrees=45
  swing1Motion=Free
  swing2LimitDegrees=45
  swing2Motion=Free
  twistLimitDegrees=45
  twistMotion=Free
 }
 {
  bone1Name=spine_03
  bone2Name=spine_01
  swing1LimitDegrees=45
  swing1Motion=Free
  swing2LimitDegrees=45
  swing2Motion=Free
  twistLimitDegrees=45
  twistMotion=Free
 }
 {bone1Name=dyn_hair_ponyTail_tipd, bone2Name=dyn_ponyTail_D, swing1LimitDegrees=40, swing2LimitDegrees=40, twistLimitDegrees=5}
 {bone1Name=dyn_ponyTail_A, bone2Name=dyn_ponyTaileBase_B, swing1LimitDegrees=25, swing2LimitDegrees=40, twistLimitDegrees=5}
 {bone1Name=dyn_ponyTail_B, bone2Name=dyn_ponyTail_A, swing1LimitDegrees=25, swing2LimitDegrees=40, twistLimitDegrees=5}
 {bone1Name=dyn_ponyTail_C, bone2Name=dyn_ponyTail_B, swing1LimitDegrees=25, swing2LimitDegrees=40, twistLimitDegrees=5}
 {bone1Name=dyn_ponyTail_D, bone2Name=dyn_ponyTail_C, swing1LimitDegrees=25, swing2LimitDegrees=40, twistLimitDegrees=5}
 {bone1Name=dyn_hair_ponyTail_tipA, bone2Name=dyn_ponyTail_D, swing1LimitDegrees=40, swing2LimitDegrees=40, twistLimitDegrees=5}
 {bone1Name=dyn_hair_ponyTail_tipb, bone2Name=dyn_ponyTail_D, swing1LimitDegrees=40, swing2LimitDegrees=40, twistLimitDegrees=5}
 {bone1Name=dyn_hair_ponyTail_tipc, bone2Name=dyn_ponyTail_D, swing1LimitDegrees=40, swing2LimitDegrees=40, twistLimitDegrees=5}
 {bone1Name=dyn_ponyTaileBase_A, bone2Name=head, swing1LimitDegrees=25, swing2LimitDegrees=40, twistLimitDegrees=5}
 {bone1Name=dyn_ponyTaileBase_B, bone2Name=dyn_ponyTaileBase_A, swing1LimitDegrees=25, swing2LimitDegrees=40, twistLimitDegrees=5}
 {bone1Name=dyn_hairSide_l, bone2Name=head, swing1LimitDegrees=15, swing2LimitDegrees=25, twistLimitDegrees=5}
 {bone1Name=dyn_hairSide_r, bone2Name=head, swing1LimitDegrees=15, swing2LimitDegrees=25, twistLimitDegrees=5}
 {
  bone1Name=dyn_earring_base_l
  bone2Name=head
  swing1LimitDegrees=45
  swing1Motion=Free
  swing2LimitDegrees=45
  swing2Motion=Free
  twistLimitDegrees=45
  twistMotion=Free
 }
 {
  bone1Name=dyn_earring_base_r
  bone2Name=head
  swing1LimitDegrees=45
  swing1Motion=Free
  swing2LimitDegrees=45
  swing2Motion=Free
  twistLimitDegrees=45
  twistMotion=Free
 }
 {bone1Name=dyn_earring_l, bone2Name=dyn_earring_base_l, swing1LimitDegrees=15, swing2LimitDegrees=5, twistLimitDegrees=5}
 {bone1Name=dyn_earring_r, bone2Name=dyn_earring_base_r, swing1LimitDegrees=15, swing2LimitDegrees=5, twistLimitDegrees=5}
 {
  bone1Name=dyn_hairFront_base
  bone2Name=head
  swing1LimitDegrees=45
  swing1Motion=Free
  swing2LimitDegrees=45
  swing2Motion=Free
  twistLimitDegrees=45
  twistMotion=Free
 }
 {bone1Name=dyn_hairFront, bone2Name=dyn_hairFront_base, swing1LimitDegrees=30, swing2LimitDegrees=5, twistLimitDegrees=5}
]
