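# Anim Blueprint for Winged Fury's face accessory mesh, driving ponytail and earring secondary motion.
# It copies the attached parent's pose, constrains earrings to weighted hair and head bones, then aims the ponytail base using spine_05, adding head-driven pitch.
# RigidBody simulates relative to spine_05, its alpha reduced when a one-time check finds a backpack.
# Head rotation is measured against spine_05; a four-node ponytail rotation chain is left unconnected.
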AnimBlueprint /BRCosmetics/Characters/Player/Female/Medium/Bodies/F_MED_Winged_Fury/Meshes/Parts/F_MED_Winged_Fury_FaceAcc_AnimBP
target_skeleton=/Game/Characters/Player/Female/Medium/Bodies/F_MED_Winged_Fury/Meshes/Parts/F_MED_Winged_Fury_FaceAcc_Skeleton
graphs [
 AnimGraph {nodes [
   n1 = Get headFwdBkRot() -> (headFwdBkRot:double)
   n2 = "Transform (Modify) Bone - Bone: dyn_ponyTaileBase_A_base"(Alpha=@n1, Translation="0,0,0", Rotation="40,0,0", Scale="1,1,1") {
    rotationMode=BMM_Additive
    rotationSpace=BCS_BoneSpace
    scaleMode=BMM_Additive
    scaleSpace=BCS_BoneSpace
    translationMode=BMM_Additive
    translationSpace=BCS_BoneSpace
    alphaBoolBlend {blendOption=Linear}
    alphaScaleBiasClamp {bMapRange=true, inRange {max=0, min=-50}, outRange {max=0, min=1}}
    boneToModify {boneName=dyn_ponyTaileBase_A_base}
    rotation {pitch=40}
   }
   n3 = "Transform (Modify) Bone - Bone: dyn_ponyTaileBase_A_base"(ComponentPose=@n2, Alpha=@n1, Translation="0,0,0", Rotation="-20,0,0", Scale="1,1,1") {
    rotationMode=BMM_Additive
    rotationSpace=BCS_BoneSpace
    scaleMode=BMM_Additive
    scaleSpace=BCS_BoneSpace
    translationMode=BMM_Additive
    translationSpace=BCS_BoneSpace
    alphaBoolBlend {blendOption=Linear}
    alphaScaleBiasClamp {bMapRange=true, inRange {max=20}}
    boneToModify {boneName=dyn_ponyTaileBase_A_base}
    rotation {pitch=-20}
   }
   n4 = Get headSideToSide() -> (headSideToSide:double)
   n5 = "Transform (Modify) Bone - Bone: dyn_ponyTaileBase_A_base"(ComponentPose=@n3, Alpha=@n4, Translation="0,0,0", Rotation="0,20,0", Scale="1,1,1") {
    rotationMode=BMM_Additive
    rotationSpace=BCS_BoneSpace
    scaleMode=BMM_Additive
    scaleSpace=BCS_BoneSpace
    translationMode=BMM_Additive
    translationSpace=BCS_BoneSpace
    alphaBoolBlend {blendOption=Linear}
    alphaScaleBiasClamp {bMapRange=true, inRange {max=40}}
    boneToModify {boneName=dyn_ponyTaileBase_A_base}
    rotation {yaw=20}
   }
   "Transform (Modify) Bone - Bone: dyn_ponyTaileBase_A_base"(ComponentPose=@n5, Alpha=@n4, Translation="0,0,0", Rotation="0,-20,0", Scale="1,1,1") {
    rotationMode=BMM_Additive
    rotationSpace=BCS_BoneSpace
    scaleMode=BMM_Additive
    scaleSpace=BCS_BoneSpace
    translationMode=BMM_Additive
    translationSpace=BCS_BoneSpace
    alphaBoolBlend {blendOption=Linear}
    alphaScaleBiasClamp {bMapRange=true, inRange {max=0, min=-40}, outRange {max=0, min=1}}
    boneToModify {boneName=dyn_ponyTaileBase_A_base}
    rotation {yaw=-20}
   }
   n6 = "Constraint - dyn_earring_base_l (PPPPPPPPPPPPPP)"!(
    ComponentPose=Local To Component(LocalPose=Copy Pose From Mesh() {bUseAttachedParent=true})
    Alpha=1
    ConstraintWeights_0=0.0051
    ConstraintWeights_1:float=0.1247
    ConstraintWeights_2:float=0.3735
    ConstraintWeights_3:float=0.0018
    ConstraintWeights_4:float=0
    ConstraintWeights_5:float=0
    ConstraintWeights_6:float=0
    ConstraintWeights_7:float=0
    ConstraintWeights_8:float=0
    ConstraintWeights_9:float=0
    ConstraintWeights_10:float=0.0024
    ConstraintWeights_11:float=0
    ConstraintWeights_12:float=0
    ConstraintWeights_13:float=0.4925
   ) -> (Pose) {
    alphaBoolBlend {blendOption=Linear}
    boneToModify {boneName=dyn_earring_base_l}
    constraintSetup [
     {targetBone {boneName=dyn_hairSide_r}}
     {targetBone {boneName=dyn_ponyTaileBase_B}}
     {targetBone {boneName=head}}
     {targetBone {boneName=dyn_ponyTaileBase_A}}
     {targetBone {boneName=dyn_hair_ponyTail_tipc}}
     {targetBone {boneName=dyn_hair_ponyTail_tipb}}
     {targetBone {boneName=dyn_hair_ponyTail_tipA}}
     {targetBone {boneName=dyn_hair_ponyTail_tipd}}
     {targetBone {boneName=dyn_hairFront}}
     {targetBone {boneName=dyn_ponyTail_D}}
     {targetBone {boneName=dyn_ponyTail_A}}
     {targetBone {boneName=dyn_ponyTail_C}}
     {targetBone {boneName=dyn_ponyTail_B}}
     {targetBone {boneName=dyn_hairSide_l}}
    ]
    constraintWeights [1,1,1,1,1,1,1,1,1,1,1,1,1,1]
   }
   n7 = "Constraint - dyn_earring_base_r (PPPPPPPPPP)"!(
    ComponentPose=@n6
    Alpha=1
    ConstraintWeights_0=0.2146
    ConstraintWeights_1:float=0.0428
    ConstraintWeights_2:float=0.7386
    ConstraintWeights_3:float=0.0001
    ConstraintWeights_4:float=0
    ConstraintWeights_5:float=0
    ConstraintWeights_6:float=0.0002
    ConstraintWeights_7:float=0
    ConstraintWeights_8:float=0
    ConstraintWeights_9:float=0.0037
   ) -> (Pose) {
    alphaBoolBlend {blendOption=Linear}
    boneToModify {boneName=dyn_earring_base_r}
    constraintSetup [
     {targetBone {boneName=dyn_hairSide_r}}
     {targetBone {boneName=dyn_ponyTaileBase_B}}
     {targetBone {boneName=head}}
     {targetBone {boneName=dyn_ponyTaileBase_A}}
     {targetBone {boneName=dyn_hairFront}}
     {targetBone {boneName=dyn_ponyTail_D}}
     {targetBone {boneName=dyn_ponyTail_A}}
     {targetBone {boneName=dyn_ponyTail_C}}
     {targetBone {boneName=dyn_ponyTail_B}}
     {targetBone {boneName=dyn_hairSide_l}}
    ]
    constraintWeights [1,1,1,1,1,1,1,1,1,1]
   }
   n8 = Look At - Bone: dyn_ponyTaileBase_A_base(ComponentPose=@n7, Alpha=1) {
    alphaBoolBlend {blendOption=Linear}
    boneToModify {boneName=dyn_ponyTaileBase_A_base}
    lookAtLocation=<0 -15>
    lookAtTarget {boneReference {boneName=spine_05}}
    lookAt_Axis {axis=<1 0>}
    lookUp_Axis {axis=<0 1>}
   }
   n9 = "Transform (Modify) Bone - Bone: dyn_ponyTaileBase_A_base"(ComponentPose=@n8, Alpha=Get headFwdBkRot(), Translation="0,0,0", Rotation="15,0,0", Scale="1,1,1") {
    rotationMode=BMM_Additive
    rotationSpace=BCS_BoneSpace
    scaleMode=BMM_Additive
    scaleSpace=BCS_BoneSpace
    translationMode=BMM_Additive
    translationSpace=BCS_BoneSpace
    alphaBoolBlend {blendOption=Linear}
    alphaScaleBiasClamp {bMapRange=true, inRange {max=20}}
    boneToModify {boneName=dyn_ponyTaileBase_A_base}
    rotation {pitch=15}
   }
   n10 = RigidBody(ComponentPose=@n9, Alpha=Get backpackOnAlpha() -> (backpackOnAlpha:double)) {
    bClampLinearTranslationLimitToRefPose=true
    bOverrideWorldGravity=true
    overridePhysicsAsset=&/F_MED_Winged_Fury_FaceAcc_Physics
    simulationSpace=BaseBoneSpace
    alphaBoolBlend {blendOption=Linear}
    baseBoneRef {boneName=spine_05}
    componentLinearAccScale=<0.4 0.4 0.4>
    componentLinearVelScale=<0.4 0.4 0.4>
    overrideWorldGravity {z=-980}
   }
   Output Pose(Result=Component To Local(ComponentPose=@n10)) {layerGroup=None}
  ]}
 EventGraph {nodes [
   Event BlueprintUpdateAnimation()
   > n1 = Sequence()
   n2 = Get playerPawn() -> (playerPawn:Fort Player Pawn Athena Object Reference)
   > Is Valid(InputObject=@n2)
   > Do Once()
   > n3 = Set IsBackpackOn(IsBackpackOn:bool="Not Equal (Object)"!(A:Object Reference=GetSkeletalMeshForPartType(self=@n2, PartType=Backpack), B:Object Reference) -> (ReturnValue:bool)) -> (Output_Get:bool)
   > n4 = Branch(Condition=@n3)
   > Set backpacOn(backpacOn:double=1) -> (Output_Get:double)
   > Set backpackOnAlpha(backpackOnAlpha:double=0.35) -> (Output_Get:double)
   Set backpacOn(execute=@n4.else, backpacOn=0)
   > Set backpackOnAlpha(backpackOnAlpha=1)
   Sequence(execute=@n1.then_1)
   n5 = GetOwnerSkeletalMeshComponent()
   n6 = BreakRotator(InRot=BreakTransform(InTransform=GetDeltaTransformFromRefPose(self=@n5, BoneName=head, BaseName=spine_05)).Rotation)
   > Set headFwdBkRot(headFwdBkRot:double=@n6.Yaw) -> (Output_Get:double)
   > Set headSideToSide(headSideToSide:double=@n6) -> (Output_Get:double)
   Event BlueprintInitializeAnimation()
   > Set MeshToCopy(MeshToCopy:Skeletal Mesh Component Object Reference=GetOwnerSkeletalMeshComponent()) -> (Output_Get:Skeletal Mesh Component Object Reference)
   n7 = TryGetPawnOwner()
   > Is Valid(InputObject=@n7)
   > n8 = Cast To FortPlayerPawnAthena(Object=@n7)
   > Set playerPawn(playerPawn:Fort Player Pawn Athena Object Reference=@n8) -> (Output_Get:Fort Player Pawn Athena Object Reference)
   TryGetPawnOwner()
  ]}
]
variables [
 backpackOnAlpha {type=number}
 backpacOn {type=number}
 headFwdBkRot {type=number}
 headSideToSide {type=number}
 IsBackpackOn {type=boolean}
 MeshToCopy {type=SkeletalMeshComponent}
 playerPawn {type=FortPlayerPawnAthena}
]
